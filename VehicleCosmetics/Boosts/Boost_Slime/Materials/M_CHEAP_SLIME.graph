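# Cheap unlit translucent Niagara sprite material for the slime vehicle boost effect.
# One slime texture supplies desaturated base shading, an alpha mask and the normal for a Fresnel rim.
# A caustics texture sampled by transformed reflection vector fakes reflections inside the mask, tinted by particle color.
# Dynamic parameters scale reflection and rim, and blend the rim from Glow_Color to square-rooted particle color via Painted.
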Material /VehicleCosmetics/Boosts/Boost_Slime/Materials/M_CHEAP_SLIME
graph {
 properties {
  bUsedWithNiagaraSprites=true
  blendMode=BLEND_Translucent
  materialDecalResponse=MDR_Color
  opacityMaskClipValue=0.85
  pixelDepthOffsetMode=PDOM_Legacy
  refractionMethod=RM_PixelNormalOffset
  shadingModel=MSM_Unlit
  usedShadingModels=Unlit
 }
 nodes [
  n1 = TextureSample() {samplerType=SAMPLERTYPE_LinearColor, texture=/VehicleCosmetics/Boosts/Boost_Slime/Textures/T_VCVFX_Boost_Slime_Slime_FB_04}
  n2 = Transform(None=ReflectionVectorWS() {bNormalizeCustomWorldNormal=true})
  n3 = DynamicParameter() {paramNames [ReflectionAmount, Brightness, FresnelBrightness, Painted]}
  n4 = Multiply(
   A=TextureSample(UVs=ComponentMask(None=@n2) {g=true, r=true}) {texture=/DealershipSharedLibrary/FX/Textures/Noise/T_VCVFX_Caustics_A}.RGB
   B=@n3.ReflectionAmount
  )
  n5 = Saturate(None=@n1.A)
  n6 = Add(A=Desaturation(None=@n1.RGB), B=Add(A=Multiply(A=@n4, B=@n5), B=Constant3Vector() {constant=<rgba 0.025 0.03 0.04 0>}))
  n7 = ParticleColor()
  n8 = Multiply(A=@n3.FresnelBrightness, B=Fresnel(Normal=Transform(None=@n1.RGB) {transformType=TRANSFORM_Local}) {baseReflectFraction=0.01, exponent=1})
  n9 = LinearInterpolate(A=VectorParameter() {parameterName=Glow_Color, defaultValue=<rgba 0.2564 2 0.0979 1>}.RGB, B=SquareRoot(None=@n7.RGB), Alpha=@n3.Painted)
  MaterialFunctionCall(Emissive=Add(A=Multiply(A=@n6, B=@n7.RGB), B=Multiply(A=@n8, B=@n9))) {materialFunction=&MF_FX_EmissiveAdjustments}
  Multiply(A=@n1.A, B=@n7.A)
  Transform(None=@n1.RGB) {transformSourceType=TRANSFORMSOURCE_Local}
 ]
}
