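# Translucent, unlit, two-sided Niagara sprite material for slime splashes in the slime vehicle boost effect.
# It sums splat and splash normal maps, transforms them and dots the result with a fixed vector to fake lighting.
# That shading, tinted by particle color, feeds an emissive adjustment function.
# Opacity combines splash alpha with squared splat red, scales by particle alpha, then passes through depth fade.
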Material /VehicleCosmetics/Boosts/Boost_Slime/Materials/M_SLIME_SPLASH
graph {
 properties {
  bUsedWithNiagaraSprites=true
  blendMode=BLEND_Translucent
  pixelDepthOffsetMode=PDOM_Legacy
  refractionMethod=RM_PixelNormalOffset
  shadingModel=MSM_Unlit
  twoSided=true
  usedShadingModels=Unlit
 }
 nodes [
  n1 = TextureCoordinate()
  n2 = TextureSample(UVs=@n1) {texture=/VehicleCosmetics/Boosts/Boost_Slime/Textures/T_VCVFX_Boost_Slime_Splat_N}
  n3 = TextureSample(UVs=@n1) {texture=/VehicleCosmetics/Boosts/Boost_Slime/Textures/T_VCVFX_Boost_Slime_Splash_N}
  n4 = Add(A=@n3.A, B=Multiply(A=Multiply(A=@n2.R, B=@n2.R), B=Constant() {r=0.5}))
  n5 = ParticleColor()
  MaterialFunctionCall(Opacity In=Saturate(None=Multiply(A=@n4, B=@n5.A)), DepthFade Distance=Constant() {r=50}) {materialFunction=&MF_VCVFX_DepthFade}
  n6 = Add(A=@n3.RGB, B=@n2.RGB)
  n7 = DotProduct(A=Transform(None=@n6), B=Constant3Vector() {constant=<rgba 0.1 0 1 0>})
  MaterialFunctionCall(Emissive=Multiply(A=ConstantBiasScale(None=@n7) {bias=1.2, desc=1.1/.2, scale=1}, B=@n5.RGB)) {materialFunction=&MF_FX_EmissiveAdjustments}
  Saturate(None=@n6)
 ]
}
