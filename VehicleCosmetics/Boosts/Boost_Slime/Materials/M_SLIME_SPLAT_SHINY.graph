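# Unlit, dithered masked Niagara sprite material that draws shiny slime splats for the Slime boost cosmetic.
# A panning flow noise texture distorts the splat texture UVs, and particle alpha times texture alpha is saturated for masking.
# A mostly desaturated stadium reflection, sampled by transformed reflection vector XY, is added to the grayscale splat, then tinted by particle color and passed through MF_FX_EmissiveAdjustments.
# Dynamic parameters drive reflection amount, brightness and distortion; Spec_Brightness is unused.
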Material /VehicleCosmetics/Boosts/Boost_Slime/Materials/M_SLIME_SPLAT_SHINY
graph {
 properties {
  bTangentSpaceNormal=false
  bUsedWithNiagaraSprites=true
  blendMode=BLEND_Masked
  ditherOpacityMask=true
  ditheredLODTransition=true
  opacityMaskClipValue=0.5
  pixelDepthOffsetMode=PDOM_Legacy
  refractionMethod=RM_IndexOfRefraction
  shadingModel=MSM_Unlit
  usedShadingModels=Unlit
 }
 nodes [
  n1 = DynamicParameter() {defaultValue=<rgba 1 1 1 0>, paramNames [Reflection_Ammount, Brightness, Spec_Brightness, UV_DISSORT]}
  n2 = Transform(None=ReflectionVectorWS())
  n3 = Saturate(None=TextureSample(UVs=ComponentMask(None=@n2) {g=true, r=true}) {texture=/VehicleCosmetics/Boosts/Boost_Slime/Textures/T_VCVFX_Boost_Slime_STADIUM_CUBEMAP}.RGB)
  n4 = Constant() {r=0.85}
  n5 = TextureCoordinate()
  n6 = Multiply(A=Time(), B=Constant() {r=1})
  n7 = ComponentMask(None=TextureSample(UVs=Panner(Coordinate=@n5, Time=@n6) {speedX=0.05, speedY=-0.2}) {texture=/DealershipSharedLibrary/FX/Textures/Noise/T_VCVFX_Flow02_N}.RGB) {g=true, r=true}
  n8 = TextureSample(UVs=Add(A=Multiply(A=@n7, B=@n1.UV_DISSORT), B=@n5)) {texture=/VehicleCosmetics/Boosts/Boost_Slime/Textures/T_VCVFX_Boost_Slime_Slime_FB_03}
  n9 = Add(A=Multiply(A=Desaturation(None=@n3, Fraction=@n4), B=@n1.Reflection_Ammount), B=Desaturation(None=@n8.RGB))
  n10 = ParticleColor()
  MaterialFunctionCall(Emissive=Multiply(A=Multiply(A=@n1.Brightness, B=@n9), B=@n10.RGB)) {materialFunction=&MF_FX_EmissiveAdjustments}
  Saturate(None=Multiply(A=@n10.A, B=@n8.A))
  Saturate(None=Transform(None=@n8.RGBA))
 ]
}
